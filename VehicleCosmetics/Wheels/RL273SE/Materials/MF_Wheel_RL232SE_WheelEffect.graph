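# Material function for the RL273SE wheel cosmetic that sets animated emissive and base color for rim and tire, with a vertex-red reroute declared for blending.
# The rim mixes parameter colors through an RGB mask, then multiplies them by two panning cloud textures for masked glow.
# The tire blends two colors using panning clouds and a world-height gradient scaled by object radius, adds inverted mask alpha, then multiplies by normal-mapped Fresnel terms.
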MaterialFunction /VehicleCosmetics/Wheels/RL273SE/Materials/MF_Wheel_RL232SE_WheelEffect
graph {nodes [
  Constant3Vector() {constant=<rgba 0.3649 0.4013 1 1>}
  Constant()
  n1 = NamedRerouteUsage()
  n2 = NamedRerouteUsage()
  n3 = NamedRerouteUsage()
  n4 = NamedRerouteUsage()
  n5 = NamedRerouteUsage()
  FunctionOutput(None=SetMaterialAttributes!(MaterialAttributes, EmissiveColor=LinearInterpolate(A=@n1, B=@n2, Alpha=@n3), BaseColor=LinearInterpolate(A=@n4, Alpha=@n5) {constB=0}) {attributeSetTypes [#1, #2]})
  FunctionOutput() {outputName=Result_1}
  GetMaterialAttributes(None=MaterialFunctionCall() {materialFunction=&MF_WheelParent_Simple}.MaterialAttributes) {attributeGetTypes [#2]}
  n6 = TextureSample(UVs=Panner(Coordinate=TextureCoordinate()) {speedX=0.33, speedY=0.1}) {samplerType=SAMPLERTYPE_Grayscale, texture=/VehicleCosmetics/Wheels/RL273SE/Textures/T_Wheel_RL273SE_clouds}
  n7 = TextureSample(UVs=Panner(Coordinate=TextureCoordinate() {uTiling=2, vTiling=2}) {speedX=-0.23, speedY=-0.146}) {samplerType=SAMPLERTYPE_Grayscale, texture=&T_Wheel_RL273SE_clouds}
  n8 = Power(Base=Clamp(None=Add(A=@n6.R, B=@n7.R)))
  n9 = VectorParameter() {parameterName=Rim_ColorC, defaultValue=<rgba 0.0703 0.8879 0.1389 1>}
  n10 = VectorParameter() {parameterName=TEAL, defaultValue=<rgba 0.0356 1 1 1>}
  n11 = TextureSampleParameter2D() {parameterName=Rim_MaskRGBA, samplerType=SAMPLERTYPE_LinearColor, texture=/VehicleCosmetics/Wheels/RL273SE/Textures/T_Wheel_RL273_SE_RGB}
  n12 = Add(A=@n11.R, B=@n11.R)
  n13 = LinearInterpolate(
   A=Constant()
   B=LinearInterpolate(
    A=LinearInterpolate(A=@n9.RGB, B=@n10.RGB, Alpha=@n12)
    B=VectorParameter() {parameterName=Rim_ColorB, defaultValue=<rgba 0 1 0 1>}.RGB
    Alpha=@n11.G
   )
   Alpha=Add(A=@n12, B=@n11.B)
  )
  n14 = TextureSampleParameter2D() {parameterName=Rim_MaskRGBA, samplerType=SAMPLERTYPE_LinearColor, texture=&T_Wheel_RL273_SE_RGB}
  NamedRerouteDeclaration(None=LinearInterpolate(A=Constant(), B=Multiply(A=@n8, B=@n13), Alpha=@n14.B)) {name=Rim_Emissive, nodeColor=<rgba 0.1016 1 0 1>}
  NamedRerouteDeclaration(None=@n13) {name=Rim_BaseColor, nodeColor=<rgba 1 0 0.2812 1>}
  n15 = VectorParameter() {parameterName=Tire_ColorA, defaultValue=<rgba 0 1 1 1>}
  n16 = VectorParameter() {parameterName=Tire_ColorB, defaultValue=<rgba 0.9823 0.0123 1 1>}
  n17 = TextureSample(UVs=Panner(Coordinate=TextureCoordinate() {uTiling=2, vTiling=0.5}) {speedX=0.33, speedY=0.1}) {samplerType=SAMPLERTYPE_Grayscale, texture=&T_Wheel_RL273SE_clouds}
  n18 = TextureSample(UVs=Panner(Coordinate=TextureCoordinate() {uTiling=3, vTiling=0.2}) {speedX=-0.23, speedY=-0.18}) {samplerType=SAMPLERTYPE_Grayscale, texture=&T_Wheel_RL273SE_clouds}
  n19 = Subtract(A=WorldPosition() {worldPositionShaderOffset=WPT_ExcludeAllShaderOffsets}.Z, B=ObjectPositionWS())
  n20 = ObjectRadius()
  n21 = ScalarParameter() {defaultValue=1.2, parameterName=Radius}
  n22 = Abs(None=Divide(A=ComponentMask(None=@n19) {b=true}, B=Multiply(A=@n20, B=@n21) {constB=1.4}))
  n23 = Add(A=Multiply(A=Add(A=@n17.R, B=@n18.R), B=@n22), B=@n22)
  n24 = LinearInterpolate(A=Multiply(A=@n15.RGB, B=@n15.RGB), B=Multiply(A=@n16.RGB, B=@n16.RGB), Alpha=Clamp(None=Power(Base=@n23) {constExponent=1.5}))
  n25 = Multiply(A=@n24, B=@n24)
  n26 = TextureSampleParameter2D() {parameterName=Tire_MaskRGBA, samplerType=SAMPLERTYPE_LinearColor, texture=/VehicleCosmetics/Wheels/RL273/Textures/T_Tire_RL273_M}
  n27 = OneMinus(None=@n26.A)
  n28 = Fresnel() {exponent=3}
  n29 = TextureSampleParameter2D() {
   group=1_Input Textures
   parameterName=Tire_normal
   samplerType=SAMPLERTYPE_LinearColor
   texture=/VehicleCosmetics/Wheels/RL273/Textures/T_Tire_RL273_N
  }
  n30 = Transform(None=@n29.RGB)
  n31 = OneMinus(None=Fresnel(Normal=@n30) {exponent=3})
  n32 = Clamp(None=ConstantBiasScale(None=Power(Base=@n31) {constExponent=200}) {bias=0, scale=2000})
  n33 = Fresnel(Normal=@n30) {exponent=3}
  n34 = Multiply(A=ScalarParameter() {defaultValue=1, parameterName=FresnelStrength}, B=Add(A=Power(Base=@n28) {constExponent=1}, B=Add(A=@n32, B=@n33)))
  NamedRerouteDeclaration(None=Multiply(A=Add(A=@n25, B=@n27), B=Add(B=@n34) {constA=0.1})) {name=Tire_Emissive, nodeColor=<rgba 0.2656 1 0 1>}
  NamedRerouteDeclaration(None=VertexColor().R) {name=VC_R, nodeColor=<rgba 1 0 0 1>}
 ]}
